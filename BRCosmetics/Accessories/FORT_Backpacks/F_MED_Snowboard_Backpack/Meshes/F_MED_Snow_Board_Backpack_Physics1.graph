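# Physics asset for the snowboard backpack cosmetic, giving collision bodies and joint limits to its dangling parts.
# A box root body anchors box bodies for two straps, two flaps and a stuff piece, plus a capsule bottle.
# Each dyn_ bone is constrained directly to root in a star layout, with narrow twist and mostly modest swing limits, widest on the right strap.
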PhysicsAsset /BRCosmetics/Accessories/FORT_Backpacks/F_MED_Snowboard_Backpack/Meshes/F_MED_Snow_Board_Backpack_Physics1
bodies {
 dyn_bottle [{shapeType=Capsule, length=5.4451, radius=2.9055, center=<3.4093 0 -0.1433>, rotation=<rot 90 0 0>}]
 dyn_flap_l [{shapeType=Box, extentX=4.1023, extentY=1.01, extentZ=6.9896, center=<0.919 0 0>}]
 dyn_flap_r [{shapeType=Box, extentX=4.9636, extentY=1.1363, extentZ=7.1999, center=<3.7774 -0.3198 -0.0307>}]
 dyn_strap_l [{shapeType=Box, extentX=4.5494, extentY=1.6797, extentZ=3.676, center=<2.6265 0.0949 -0.0341>}]
 dyn_strap_r [{shapeType=Box, extentX=4.2952, extentY=1.5578, extentZ=3.9371, center=<1.5711 -0.5039 0.3609>}]
 dyn_stuff [{shapeType=Box, extentX=11.5103, extentY=7.3753, extentZ=3.8412, center=<0.5392 0.2143 0.1033>}]
 root [{shapeType=Box, extentX=16.7307, extentY=8.5102, extentZ=20.0917, center=<-2.0195 -7.789 -0.8046>, rotation=<rot 0 8.426 0>}]
}
constraints [
 {bone1Name=dyn_strap_r, bone2Name=root, swing1LimitDegrees=60, swing2LimitDegrees=5, twistLimitDegrees=5}
 {bone1Name=dyn_strap_l, bone2Name=root, swing1LimitDegrees=25.7143, swing2LimitDegrees=5, twistLimitDegrees=5}
 {bone1Name=dyn_flap_l, bone2Name=root, swing1LimitDegrees=20, swing2LimitDegrees=20, twistLimitDegrees=5}
 {bone1Name=dyn_flap_r, bone2Name=root, swing1LimitDegrees=20, swing2LimitDegrees=10, twistLimitDegrees=5}
 {bone1Name=dyn_bottle, bone2Name=root, swing1LimitDegrees=20, swing2LimitDegrees=20, twistLimitDegrees=5}
 {bone1Name=dyn_stuff, bone2Name=root, swing1LimitDegrees=10, swing2LimitDegrees=15, twistLimitDegrees=5}
]
